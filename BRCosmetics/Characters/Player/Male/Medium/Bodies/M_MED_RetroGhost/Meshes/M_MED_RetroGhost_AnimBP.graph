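# Post-copy animation Blueprint for the RetroGhost medium male body cosmetic, following the attached parent mesh's pose.
# It offsets support_neck, runs a Control Rig fed the emote-playing flag, then simulates both grenades with AnimDynamics.
# The cached result feeds a weighted spine, clavicle, pec and upper-arm constraint on the zippo base, plus zippo AnimDynamics.
# Finally, it adjusts attach_cape only when the current LOD is above zero, checked each update.
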AnimBlueprint /BRCosmetics/Characters/Player/Male/Medium/Bodies/M_MED_RetroGhost/Meshes/M_MED_RetroGhost_AnimBP
target_skeleton=&/M_MED_RetroGhost_Skeleton
graphs [
 AnimGraph {nodes [
   n1 = "Constraint - dyn_zippo_base (PPPPPPPP)"!(
    ComponentPose=Local To Component(LocalPose=Use cached pose 'tyZippo'())
    Alpha=1
    ConstraintWeights_0=0.0859
    ConstraintWeights_1:float=0.1917
    ConstraintWeights_2:float=0.3754
    ConstraintWeights_3:float=0.1745
    ConstraintWeights_4:float=0.0609
    ConstraintWeights_5:float=0.1116
    ConstraintWeights_6:float=0
    ConstraintWeights_7:float=0
   ) -> (Pose) {
    alphaBoolBlend {blendOption=Linear}
    boneToModify {boneName=dyn_zippo_base}
    constraintSetup [
     {targetBone {boneName=spine_02}}
     {targetBone {boneName=spine_03}}
     {targetBone {boneName=spine_04}}
     {targetBone {boneName=spine_05}}
     {targetBone {boneName=clavicle_l}}
     {targetBone {boneName=deform_pec_l}}
     {targetBone {boneName=deform_pec_r}}
     {targetBone {boneName=deform_upperarm_fr_l}}
    ]
    constraintWeights [1,1,1,1,1,1,1,1]
   }
   n2 = Anim Dynamics - Bone: dyn_zippo_base_01(ComponentPose=@n1, Alpha=1, ExternalForce="0,0,0") {
    angularDampingOverride=0.9
    bOverrideAngularDamping=true
    bUseGravityOverride=true
    simulationSpace=RootRelative
    alphaBoolBlend {blendOption=Linear}
    boundBone {boneName=dyn_zippo_base_01}
    componentAppliedLinearAccClamp=<5000 5000 5000>
    gravityOverride {z=-980}
    physicsBodyDefinitions [{
      boundBone {boneName=dyn_zippo_base_01}
      boxExtents=<2 2 2>
      constraintSetup {angularTargetAxis=AxisZ, twistAxis=AxisZ, angularLimitsMax=<0 20 0>, angularLimitsMin=<-10 -20 0>, angularTarget=<0 0 0>}
      localJointOffset=<0 0.4 0>
     }]
   }
   n3 = "Transform (Modify) Bone - Bone: attach_cape"!(ComponentPose=@n2, bAlphaBoolEnabled:bool=Get Is LOD1() -> (Is LOD1:bool), Translation="0,-3,0", Rotation="0,0,11", Scale="1,1,1") -> (Pose) {
    alphaInputType=Bool
    rotationMode=BMM_Replace
    rotationSpace=BCS_BoneSpace
    translationMode=BMM_Additive
    translationSpace=BCS_BoneSpace
    alphaBoolBlend {blendOption=Linear}
    boneToModify {boneName=attach_cape}
    rotation {roll=11}
    translation {y=-3}
   }
   Output Pose(Result=Component To Local(ComponentPose=@n3)) {layerGroup=None}
   n4 = "Transform (Modify) Bone - Bone: support_neck"(
    ComponentPose=Local To Component(LocalPose=Copy Pose From Mesh() {bUseAttachedParent=true})
    Alpha=1
    Translation="-0.000,0.000,-1.810"
    Rotation="21.604,16.456,54.236"
    Scale="1,1,1"
   ) {
    rotationMode=BMM_Additive
    translationMode=BMM_Additive
    alphaBoolBlend {blendOption=Linear}
    boneToModify {boneName=support_neck}
    rotation=<rot 21.604 16.456 54.236>
    translation {z=-1.81}
   }
   n5 = Control Rig!(Source=Component To Local(ComponentPose=@n4), Alpha, Is Playing Emote:bool=Get bIsPlayingEmote()) -> (Pose) {
    alphaBoolBlend {blendOption=Linear}
    controlRigAssetReference {blueprintRigClass=&/M_MED_RetroGhost_CtrlRig.M_MED_RetroGhost_CtrlRig_C}
   }
   n6 = Anim Dynamics - Bone: dyn_grenade_l_01(ComponentPose=Local To Component(LocalPose=@n5), Alpha=1, ExternalForce="0,0,0") {
    angularDampingOverride=0.9999
    bOverrideAngularDamping=true
    bOverrideLinearDamping=true
    bUseGravityOverride=true
    linearDampingOverride=0.9999
    simulationSpace=RootRelative
    alphaBoolBlend {blendOption=Linear}
    boundBone {boneName=dyn_grenade_l_01}
    componentAppliedLinearAccClamp=<5000 5000 5000>
    gravityOverride {z=-980}
    physicsBodyDefinitions [{
      boundBone {boneName=dyn_grenade_l_01}
      boxExtents=<4 4 4>
      constraintSetup {angularLimitsMax=<0 0 5>, angularLimitsMin=<0 -20 -5>, angularTarget=<0 0 0>, linearAxesMax=<0.2 0 0>, linearAxesMin=<-0.2 0 0>}
      localJointOffset=<0.3 0 0>
     }]
   }
   n7 = Anim Dynamics - Bone: dyn_grenade_r_01(ComponentPose=@n6, Alpha=1, ExternalForce="0,0,0") {
    angularDampingOverride=0.9999
    bOverrideAngularDamping=true
    bOverrideLinearDamping=true
    bUseGravityOverride=true
    linearDampingOverride=0.9999
    simulationSpace=RootRelative
    alphaBoolBlend {blendOption=Linear}
    boundBone {boneName=dyn_grenade_r_01}
    componentAppliedLinearAccClamp=<5000 5000 5000>
    gravityOverride {z=-980}
    physicsBodyDefinitions [{
      boundBone {boneName=dyn_grenade_r_01}
      boxExtents=<4 4 4>
      constraintSetup {angularLimitsMax=<0 0 5>, angularLimitsMin=<0 -20 -5>, angularTarget=<0 0 0>, linearAxesMax=<0.2 0 0>, linearAxesMin=<-0.2 0 0>}
      localJointOffset=<0.3 0 0>
     }]
   }
   Save cached pose 'tyZippo'(Pose=Component To Local(ComponentPose=@n7)) {cacheName=tyZippo}
  ]}
 EventGraph {nodes [
   Event BlueprintUpdateAnimation()
   > Set Is LOD1(Is LOD1:bool="integer > integer"(A=Get CurrentLOD())) -> (Output_Get:bool)
  ]}
]
variables [Is LOD1 {type=boolean}]
